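# Volumetric cloud material that renders a layered sky of several cloud types, with a StormClouds blend for darker clouds and animated lightning.
# Global layout textures drive altitude-profile lookups, broken up by wind-animated 3D noise with distortion and optional extra detail.
# These terms pack into ConservativeDensity, then combine with a coverage mask to form extinction, albedo, occlusion and lightning emission.
# Dual-phase, two-octave multiscattering shapes lighting, and mobile uses simplified outputs.
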Material /Engine/EngineSky/VolumetricClouds/m_SimpleVolumetricCloud
graph {
 properties {
  bUsedWithVolumetricCloud=true
  blendMode=BLEND_Additive
  materialDomain=MD_Volume
  pixelDepthOffsetMode=PDOM_Legacy
  usedShadingModels=Default Lit
 }
 nodes [
  n1 = ComponentMask(None=Subtract(A=NamedRerouteUsage()) {constB=0.5}) {g=true, r=true}
  n2 = VectorParameter() {
   desc="This set of parameters control the scale of the CloudGlobalPattern texture for each of the cloud types  ( r = stratocumulus, g = altostratus, and b = cirrostratus, a = nimbostratus )"
   group=00 - Cloud Layout
   parameterName=Layout_CloudPerTypeScale
   sortPriority=40
   channelNames {a="Nimbostratus (Storm)", b=Cirrostratus, g=Altostratus, r=Stratocumulus}
   defaultValue=<rgba 1 1 1 1>
  }
  n3 = NamedRerouteUsage()
  n4 = TextureSample(UVs=Multiply(A=@n1, B=@n2.Stratocumulus), Tex=@n3) {samplerSource=SSM_Wrap_WorldGroupSettings, samplerType=SAMPLERTYPE_LinearColor}
  n5 = TextureSample(UVs=Multiply(A=@n2.Altostratus, B=@n1), Tex=@n3) {samplerSource=SSM_Wrap_WorldGroupSettings, samplerType=SAMPLERTYPE_LinearColor}
  n6 = TextureSample(UVs=Multiply(A=@n2.Cirrostratus, B=@n1), Tex=@n3) {samplerSource=SSM_Wrap_WorldGroupSettings, samplerType=SAMPLERTYPE_LinearColor}
  n7 = TextureSample(UVs=Multiply(A=@n2."Nimbostratus (Storm)", B=@n1), Tex=@n3) {samplerSource=SSM_Wrap_WorldGroupSettings, samplerType=SAMPLERTYPE_LinearColor}
  n8 = MaterialFunctionCall(X=@n4.R, Y=@n5.G, Z=@n6.B, A=@n7.A) {materialFunction=&MakeFloat4}
  n9 = VectorParameter() {
   desc="This set of parameters control the visibility for each of the different cloud types individually ( r = stratocumulus, g = altostratus, and b = cirrostratus, a = nimbostratus ) "
   group=00 - Cloud Layout
   parameterName=Layout_CloudType
   sortPriority=30
   channelNames {a="Nimbostratus (Storm)", b=Cirrostratus, g=Altostratus, r=Stratocumulus}
   defaultValue=<rgba 1 1 1 2>
  }
  n10 = MaterialFunctionCall(Float3=Multiply(A=ComponentMask(None=@n8.Result) {b=true, g=true, r=true}, B=@n9)) {materialFunction=&BreakOutFloat3Components}
  n11 = Max(A=Max(A=@n10.R, B=@n10.R), B=@n10.B)
  n12 = ComponentMask(None=@n8.Result) {a=true}
  n13 = ScalarParameter() {
   desc="This parameter controls the strength of the blend from normal clouds to storm clouds, which have additional controls for albedo, occlusion,  distortion, and (optional) lightning effects.  "
   group=02 - Storm
   parameterName=StormClouds
   sliderMax=1
   sortPriority=1
  }
  NamedRerouteDeclaration(None=LinearInterpolate(A=@n11, B=Max(A=@n11, B=@n12), Alpha=@n13)) {name=ConservativeDensity.r, nodeColor=<rgba 0.3359 1 0 1>}
  NamedRerouteDeclaration(None=LinearInterpolate(A=@n11, B=@n12, Alpha=@n13)) {name=ConservativeDensity.g, nodeColor=<rgba 0.875 1 0 1>}
  n14 = TextureObjectParameter() {
   desc="This texture describes the shape of the clouds as the altitue changes.  Each channel of this texture describes the profile shape and relative altitude of a different cloud type ( r = stratocumulus, g = altostratus, and b = cirrostratus, a = nimbostratus )"
   group=00 - Cloud Layout
   parameterName=Layout_CloudHeightProfile
   samplerType=SAMPLERTYPE_LinearColor
   sortPriority=0
   texture=/Engine/Functions/Engine_MaterialFunctions02/PivotPainter2/Black_1x1_EXR_Texture
  }
  NamedRerouteDeclaration(None=@n14) {name=ProfileObject, nodeColor=<rgba 0 1 0.0859 1>}
  n15 = TextureObjectParameter() {
   desc="This texture can be used to add clouds to or remove clouds from the global pattern texture in the masked region.  Each channel of this texture applies to a different cloud type ( r = stratocumulus, g = altostratus, and b = cirrostratus ).  The alpha channel is unused."
   group=00 - Cloud Layout
   parameterName=Layout_GlobalCloudMask
   samplerType=SAMPLERTYPE_LinearColor
   sortPriority=20
   texture=&Black_1x1_EXR_Texture
  }
  NamedRerouteDeclaration(None=@n15) {name=MaskObject, nodeColor=<rgba 1 0 0.1406 1>}
  n16 = TextureObjectParameter() {
   desc="This texture defines the world location for each type of cloud where each channel describes a different cloud type  ( r = stratocumulus, g = altostratus, and b = cirrostratus, a = nimbostratus )."
   group=00 - Cloud Layout
   parameterName=Layout_CloudGlobalPattern
   samplerType=SAMPLERTYPE_LinearColor
   sortPriority=10
   texture=&Black_1x1_EXR_Texture
  }
  NamedRerouteDeclaration(None=@n16) {name=LayoutObject, nodeColor=<rgba 1 0.5625 0 1>}
  n17 = VectorParameter() {
   desc="This vector controls the offset and rotation (about the world z axis) of the layout region textures (the GlobalCloudMask, and the GlobalCloudPattern)."
   group=00 - Cloud Layout
   parameterName=Layout_GlobalTexturePlacement
   sortPriority=50
   channelNames {a=SkyTextureRotation, b=Position.Z, g=Position.Y, r=Position.X}
  }
  NamedRerouteDeclaration(None=@n17) {name=SkyPlacement, nodeColor=<rgba 0 1 0.8828 1>}
  n18 = CloudSampleAttribute()
  n19 = NamedRerouteUsage()
  n20 = TextureSample!(UVs=AppendVector(A=@n18.NormAltitudeInLayer, B=@n4.R), Tex=@n19, Level, Apply View MipBias) {constMipValue=0, mipValueMode=TMVM_MipLevel, samplerSource=SSM_Wrap_WorldGroupSettings, samplerType=SAMPLERTYPE_LinearColor}
  n21 = TextureSample(UVs=AppendVector(A=@n18.NormAltitudeInLayer, B=@n5.G), Tex=@n19) {constMipValue=0, mipValueMode=TMVM_MipLevel, samplerSource=SSM_Wrap_WorldGroupSettings, samplerType=SAMPLERTYPE_LinearColor}
  n22 = TextureSample(UVs=AppendVector(A=@n18.NormAltitudeInLayer, B=@n6.B), Tex=@n19) {constMipValue=0, mipValueMode=TMVM_MipLevel, samplerSource=SSM_Wrap_WorldGroupSettings, samplerType=SAMPLERTYPE_LinearColor}
  n23 = MaterialFunctionCall(Float3=Multiply(
    A=@n9
    B=MaterialFunctionCall(X=@n20.R, Y=@n21.G, Z=@n22.B) {materialFunction=&MakeFloat3}.Result
   )) {materialFunction=&BreakOutFloat3Components}
  n24 = Max(A=Max(A=@n23.R, B=@n23.R), B=@n23.B)
  n25 = TextureSample(UVs=AppendVector(A=@n18.NormAltitudeInLayer, B=@n7.A), Tex=@n19) {constMipValue=0, mipValueMode=TMVM_MipLevel, samplerSource=SSM_Wrap_WorldGroupSettings, samplerType=SAMPLERTYPE_LinearColor}
  n26 = Multiply(A=@n9."Nimbostratus (Storm)", B=@n25.A)
  NamedRerouteDeclaration(None=Subtract(A=LinearInterpolate(A=@n24, B=@n26, Alpha=@n13)) {constB=0.5}) {name=ConservativeDensity.b, nodeColor=<rgba 1 0.8438 0 1>}
  n27 = VectorParameter() {
   desc="This parameter set controls both the scale of the green channel of the Noise Texture as well as how strong an effect WindVelocity has on the this noise sample.  Noise1 controls the primary breakup noise for the clouds."
   group=01 - Cloud Shape
   parameterName=Noise1_Coordinates
   sortPriority=40
   channelNames {a=WindMultiplier, b=Scale.Z, g=Scale.Y, r=Scale.X}
   defaultValue=<rgba 4.167 4.444 9.091 7>
  }
  n28 = Multiply(A=@n27, B=NamedRerouteUsage())
  n29 = NamedRerouteUsage()
  n30 = Multiply(A=ComponentMask(None=@n29) {b=true, g=true, r=true}, B=ComponentMask(None=@n29) {a=true})
  n31 = Multiply(A=@n27.WindMultiplier, B=@n30)
  n32 = Constant() {r=0.05}
  n33 = Add(A=Multiply(A=CloudSampleAttribute().NormAltitudeInLayer, B=OneMinus(None=@n32)), B=@n32)
  n34 = Multiply(A=Abs(None=Subtract(A=@n33) {constB=0.5})) {constB=2}
  n35 = MaterialFunctionCall(X=@n33, Y=@n33, Z=Multiply(A=@n34, B=@n34)) {materialFunction=&MakeFloat3}
  n36 = VectorParameter() {
   desc="This parameter set controls both the scale of the red channel of the Noise Texture as well as how strong an effect WindVelocity has on the this noise sample.  Noise2 controls the amount and scale of distortion for Noise1."
   group=01 - Cloud Shape
   parameterName=Noise2_Coordinates
   sortPriority=50
   channelNames {a=WindMultiplier, b=Scale.Z, g=Scale.Y, r=Scale.X}
   defaultValue=<rgba 60 60 75 -6>
  }
  n37 = Add(A=Multiply(A=@n36, B=NamedRerouteUsage()), B=Multiply(A=@n36.WindMultiplier, B=@n30) {constB=-2})
  n38 = NamedRerouteUsage()
  n39 = Constant()
  n40 = VectorParameter() {
   desc="Each parameter in this set is subtracted the respective NoiseTexture sample (and may be positive or negative).  This determines how the noise texture dilates or erodes the Cloud Layout without increasing the overall contrast of the noise. "
   group=01 - Cloud Shape
   parameterName=Noise_Bias
   sortPriority=20
   channelNames {a=N / A, b=Noise3, g=Noise2, r=Noise1}
   defaultValue=<rgba 0.5 0.8 0.5 0>
  }
  n41 = VectorParameter() {
   desc="Each parameter in this set is multiplied against the respective result of the biased NoiseTexture sample (and may be positive or negative).  This increase or decreases the contrast of the noise and thus increases or decreases the apparent strength of the noise effect on the Cloud Layout."
   group=01 - Cloud Shape
   parameterName=Noise_Strength
   sortPriority=30
   channelNames {a=StormyDistortFactor, b=Noise3, g=Noise2, r=Noise1}
   defaultValue=<rgba 0.8 0.08 0.03 2.5>
  }
  n42 = ScalarParameter() {
   desc="This parameter controls the strength of the blend from normal clouds to storm clouds, which have additional controls for albedo, occlusion,  distortion, and (optional) lightning effects.  "
   group=02 - Storm
   parameterName=StormClouds
   sliderMax=1
   sortPriority=1
  }
  n43 = Multiply(
   A=Subtract(A=TextureSample(UVs=@n37, Tex=@n38, Level=@n39) {mipValueMode=TMVM_MipLevel, samplerSource=SSM_Wrap_WorldGroupSettings}.R, B=@n40.Noise2) {constB=0.5}
   B=LinearInterpolate(A=@n41.Noise2, B=Multiply(A=@n41.Noise2, B=@n41.StormyDistortFactor) {constB=3}, Alpha=@n42)
  )
  n44 = TextureSample(UVs=Add(A=Add(A=@n28, B=@n31), B=Multiply(A=@n35.Result, B=@n43)), Tex=NamedRerouteUsage(), Level=Constant()) {mipValueMode=TMVM_MipLevel, samplerSource=SSM_Wrap_WorldGroupSettings}
  n45 = ScalarParameter() {
   defaultValue=-0.2
   desc="This parameter determines the overall amount of cloud cover where positive values increase the amount of cloud cover, and negative values decrease the amount of cloud cover."
   group=00 - Cloud Layout
   parameterName=Cloud_GlobalCoverage
   sliderMax=2
   sliderMin=-2
   sortPriority=100
  }
  n46 = Multiply(A=Add(A=Subtract(A=@n44.G, B=@n40.Noise1) {constB=0.5}, B=@n45), B=@n41.Noise1)
  n47 = VectorParameter() {
   desc="This parameter controls the multiplicative contribution to the overall cloud noise of each of the channels, as well as the total magnitude of the multiplicative effect (via the alpha channel).   In other words the cloud noise is multiplied by a factor determined by this formula:  MultChannel.a * (Noise.r * MultChannel.r + Noise.g * MultChannel.g + Noise.b * MultChannel.b)."
   group=01 - Cloud Shape
   parameterName=Noise3_MultChannel
   sortPriority=70
   channelNames {a=MultStrength, b="", g="", r=""}
   defaultValue=<rgba 0 0 1 0.4>
  }
  n48 = VectorParameter() {
   desc="This parameter set controls both the scale of the blue channel of the Noise Texture as well as how strong an effect WindVelocity has on the this noise sample.  Noise3 is added to Noise1 as an additional level of detail.  Noise3 also has an additional control (the MultChannel) which is used to modulate the existing noise by any combination of NoiseTexture channels."
   group=01 - Cloud Shape
   parameterName=Noise3_Coordinates
   sortPriority=60
   channelNames {a=WindMultiplier, b=Scale.Z, g=Scale.Y, r=Scale.X}
   defaultValue=<rgba 30 40 25 8>
  }
  n49 = NamedRerouteUsage()
  n50 = TextureSample(UVs=Add(A=Multiply(A=@n48, B=@n49), B=Multiply(A=@n48.WindMultiplier, B=@n30) {constB=1.5}), Tex=NamedRerouteUsage(), Level=Constant()) {mipValueMode=TMVM_MipLevel, samplerSource=SSM_Wrap_WorldGroupSettings}
  n51 = ScalarParameter() {
   desc="This parameter controls the strength of the blend from normal clouds to storm clouds, which have additional controls for albedo, occlusion,  distortion, and (optional) lightning effects.  "
   group=02 - Storm
   parameterName=StormClouds
   sliderMax=1
   sortPriority=1
  }
  n52 = Multiply(A=LinearInterpolate(A=DotProduct(A=@n47, B=@n50.RGB), B=@n50.G, Alpha=@n51), B=@n47.MultStrength)
  n53 = Add(A=@n45, B=Subtract(A=@n50.B, B=@n40.Noise3) {constB=0.5})
  n54 = StaticSwitchParameter(True=Add(A=Multiply(A=@n46, B=@n52) {desc=Combine Noise3 MultChannel}, B=Multiply(A=@n41.Noise3, B=@n53)) {desc=Add Noise3}, False=@n46) {
   desc="Enable the use of a third sample of the NoiseTexture3D for breaking up the cloud shapes with an additional level of detail (and/or different method of breakup)."
   group=01 - Cloud Shape
   parameterName=UseNoise3
   sortPriority=10
  }
  NamedRerouteDeclaration(None=@n54) {name=ConservativeDensity.a, nodeColor=<rgba 1 0 0.375 1>}
  n55 = VectorParameter() {
   desc="This vector controls the strength of the wind along each world axis (x, y, or z), the value for which may be positive or negative.  The alpha channel can be used to uniformly increase or decrease the strength of the wind."
   group=00 - Cloud Layout
   parameterName=Layout_WindControls
   sortPriority=80
   channelNames {a=VelocityScale, b=Velocity.Z, g=Velocity.Y, r=Velocity.X}
   defaultValue=<rgba 1 1 0.5 0.3333>
  }
  n56 = Multiply(A=@n55, B=@n55)
  n57 = Time() {bOverride_Period=true, period=9999999}
  n58 = TextureProperty(Tex=@n16)
  n59 = ScalarParameter() {
   defaultValue=256
   desc="This parameter controls the distance over which the layout region textures ( GlobalCloudMask, and CloudGlobalPattern ) repeat (in km)."
   group=00 - Cloud Layout
   parameterName=Layout_CloudGlobalScale
   sliderMax=1024
   sliderMin=1
   sortPriority=35
  }
  n60 = Multiply(A=Divide(A=ComponentMask(None=@n58) {r=true}, B=@n59)) {constB=1000}
  NamedRerouteDeclaration(None=AppendVector(A=ComponentMask(None=@n56) {b=true, g=true, r=true}, B=Divide(A=@n57, B=@n60))) {name=CloudWind, nodeColor=<rgba 1 0 0.7266 1>}
  n61 = WorldPosition()
  n62 = MaterialFunctionCall!(Km=@n59) {
   desc="scaled for clouds at a max height of 8km  maxVisDist = sqrt(h^2+2(6371*h))"
   materialFunction=/Engine/Functions/Engine_MaterialFunctions01/Units/Km-to-Cm
  }
  n63 = Add(A=TruncateLWC!(None=Divide(A=@n61.XYZ, B=@n62.Cm)), B=Constant3Vector() {constant=<rgba 0.5 0.5 0 0>})
  n64 = Rotator(Coordinate=ComponentMask(None=@n63) {g=true, r=true}, Time=@n17.SkyTextureRotation) {speed=1}
  n65 = ComponentMask(None=@n63) {b=true}
  NamedRerouteDeclaration(None=Subtract(A=AppendVector(A=@n64, B=@n65), B=@n17)) {name=SkyTexScale, nodeColor=<rgba 0 1 0.9531 1>}
  n66 = TextureObjectParameter() {
   desc="This texture is used to break up the shapes of the clouds for authoring their general form.  Each channel of the texture describes a different noise pattern the influence of which is determined by the Noise controls (r = Noise1, g = Noise2, and b = Noise3).  The texture has an alpha channel which is unused in the material. "
   group=01 - Cloud Shape
   parameterName=Noise_Texture3D
   sortPriority=0
   texture=/Engine/EngineResources/DefaultVolumeTexture
  }
  NamedRerouteDeclaration(None=@n66) {name=NoiseObject, nodeColor=<rgba 0 1 0.7422 1>}
  n67 = VectorParameter() {
   desc="This is the apparent color of the clouds when they are not storm clouds.  The alpha channel of the parameter determines the strength of light occlusion for the material (and consequently how dense the shadows appear). "
   group=01 - Cloud Shape
   parameterName=Cloud_AlbedoColor
   sortPriority=80
   channelNames {a=AmbientOcclusion, b="", g="", r=""}
   defaultValue=<rgba 0.98 0.98 0.98 0.5>
  }
  n68 = VectorParameter() {
   desc="This is the apparent color of the clouds when they are storm clouds.  The alpha channel of the parameter determines the multiplicative factor for the shadow density (lower values result in darker shadows)."
   group=02 - Storm
   parameterName=Storm_AlbedoColor
   sortPriority=60
   channelNames {a=StormOcclusionFactor, b="", g="", r=""}
   defaultValue=<rgba 0.3 0.3375 0.375 0.3333>
  }
  n69 = CloudSampleAttribute()
  n70 = Constant() {r=0.5}
  n71 = Constant()
  n72 = Constant() {r=1}
  n73 = VectorParameter() {
   desc="This set of parameters specifies how strongly the GlobalCloudMask texture affects each of the cloud types ( r = stratocumulus, g = altostratus, and b = cirrostratus ).  The alpha channel of this set determines the overall strength of the effect."
   group=00 - Cloud Layout
   parameterName=Layout_CloudTypeMask
   sortPriority=90
   channelNames {a=MaskContribution, b=Cirrostratus, g=Altostratus, r=Stratocumulus}
   defaultValue=<rgba 0 0 0 1>
  }
  n74 = ComponentMask(None=Add(A=NamedRerouteUsage(), B=NamedRerouteUsage())) {g=true, r=true}
  n75 = NamedRerouteUsage()
  n76 = DotProduct(
   A=ComponentMask(None=@n73) {b=true, g=true, r=true}
   B=Multiply(
    A=TextureSample!(UVs=@n74, Tex=@n75, Apply View MipBias) {samplerSource=SSM_Wrap_WorldGroupSettings, samplerType=SAMPLERTYPE_LinearColor}.RGB
    B=@n9
   )
  )
  n77 = Multiply(A=@n73.MaskContribution) {constA=0.1, constB=0.01}
  n78 = MaterialFunctionCall(Float4=VolumetricAdvancedMaterialInput().ConservativeDensity as Float4) {materialFunction=&BreakOutFloat4Components}
  n79 = Add(A=@n78.B, B=@n78.B) {desc=Blend Profiles + Volume Noise}
  n80 = ScalarParameter() {
   desc="This parameter controls the strength of the blend from normal clouds to storm clouds, which have additional controls for albedo, occlusion,  distortion, and (optional) lightning effects.  "
   group=02 - Storm
   parameterName=StormClouds
   sliderMax=1
   sortPriority=1
  }
  n81 = Multiply(B=@n80) {constA=0.6667}
  n82 = ScalarParameter() {
   defaultValue=0.008
   desc="This parameter controls the overall density of the cloud material, rendering them more opaque at higher values or hazier at lower values."
   group=01 - Cloud Shape
   parameterName=Cloud_GlobalDensity
   sliderMax=1
   sortPriority=90
  }
  n83 = Multiply(
   A=Add(A=Multiply(A=@n76, B=@n77), B=Multiply(A=@n78.G, B=@n79) {constA=1, desc=Weight By 2d Cloud Texture}) {desc=Add additive mask}
   B=LinearInterpolate(A=@n82, B=Add(A=@n82, B=@n81), Alpha=@n80)
  ) {desc=Global Density}
  n84 = WorldPosition() {worldPositionShaderOffset=WPT_CameraRelative}
  n85 = Saturate(None=Multiply(A=Length(None=@n84.XYZ) {is=class MaterialExpressionLength}) {constB=0})
  n86 = Saturate(None=Multiply(A=If(A=@n69.NormAltitudeInLayer, B=@n70, "A > B"=@n71, "A == B"=@n72, "A < B"=@n71) {equalsThreshold=0.48}, B=Multiply(A=@n83, B=@n85)))
  n87 = Multiply(A=Multiply(A=@n86, B=@n86), B=@n86)
  n88 = ScalarParameter() {
   desc="This parameter controls the strength of the blend from normal clouds to storm clouds, which have additional controls for albedo, occlusion,  distortion, and (optional) lightning effects.  "
   group=02 - Storm
   parameterName=StormClouds
   sortPriority=1
  }
  n89 = LinearInterpolate(A=@n67, B=@n68, Alpha=Multiply(A=OneMinus(None=@n87), B=@n88))
  ShadingPathSwitch(Default=@n89, Deferred=@n89, Mobile=@n67)
  n90 = VectorParameter() {
   desc="This parameter controls the color and (using the alpha channel) the brightness of the lightning flashes. "
   group=02 - Storm
   parameterName=Storm_LightningColor
   sortPriority=40
   channelNames {a=BrightnessMult, b="", g="", r=""}
   defaultValue=<rgba 2 1.8 2 10>
  }
  n91 = Multiply(A=@n90, B=@n90)
  n92 = Multiply(A=Power(Base=Constant() {r=4}, Exp=Multiply(A=@n86) {constB=-1}), B=@n86)
  n93 = ScalarParameter() {
   desc="This parameter controls the strength of the blend from normal clouds to storm clouds, which have additional controls for albedo, occlusion,  distortion, and (optional) lightning effects.  "
   group=02 - Storm
   parameterName=StormClouds
   sliderMax=1
   sortPriority=1
  }
  n94 = ScalarParameter() {
   defaultValue=6
   desc="This parameter controls the relative scale of the lighting texture (which is used to determine the location of lighting flashes in the material)."
   group=02 - Storm
   parameterName=Storm_LightningTexScale
   sliderMax=32
   sliderMin=0.0312
   sortPriority=10
  }
  n95 = Multiply(A=Multiply(A=@n2."Nimbostratus (Storm)", B=NamedRerouteUsage()), B=@n94) {constB=6}
  n96 = Time() {bOverride_Period=true, period=9999999}
  n97 = Sine(None=@n96)
  n98 = VectorParameter() {
   desc="This set of parameters is used to control the animation of lightning in the material.  The lighting flicker adds a secondary animation to the flashes.  Dynamic lightning animation is used to determine the frequency of lightning flashes (i.e. the higher this value the more frequently the flashes occur).  Manual Lightning is intended to be used when Dynamic Lighting is set to 0, and will let the user specify the state of the flashes independent of time (mostly useful when rendering stills or when manually updating time in a sequence)."
   group=02 - Storm
   parameterName=Storm_LightningAnim
   sortPriority=20
   channelNames {a=N / A, b=ManualLightningAnim, g=DynamicLightingAnim, r=LightningFlicker}
   defaultValue=<rgba 0 0.1 0 1>
  }
  n99 = Multiply(A=@n95, B=Multiply(A=Multiply(A=@n97) {constB=0.001}, B=@n98.LightningFlicker))
  n100 = AppendVector(
   A=ComponentMask(None=Add(A=@n95, B=@n99)) {g=true, r=true}
   B=Add(A=Multiply(A=@n96, B=@n98.DynamicLightingAnim), B=@n98.ManualLightningAnim)
  )
  n101 = TextureSample(UVs=@n100) {samplerSource=SSM_Wrap_WorldGroupSettings, samplerType=SAMPLERTYPE_Grayscale, texture=&/VT_Lightning}
  n102 = Multiply(A=@n101.R, B=@n101.R)
  n103 = VectorParameter() {
   desc="This set of parameters controls the effect of lightning on the clouds.  Source Power controls the intensity of the core of the lighting strike.  Fill Scatter controls the distance the lightning scatters in the clouds.  Fill Scatter intensity controls the brightness of the scattered light.  Lastly the SecondMipLevel controls the sharpness of the details contributed by the second lighting texture sample (where higher values are sharper)."
   group=02 - Storm
   parameterName=Storm_LightningClouds
   sortPriority=30
   channelNames {a=SecondMipLevel, b=FillScatterIntensity, g=FillScatter, r=SourcePower}
   defaultValue=<rgba 6 0.05 0.175 2.5>
  }
  n104 = TextureSample!(UVs=@n100, Tex, Level=@n103.SecondMipLevel, Apply View MipBias) {
   mipValueMode=TMVM_MipLevel
   samplerSource=SSM_Wrap_WorldGroupSettings
   samplerType=SAMPLERTYPE_Grayscale
   texture=&VT_Lightning
  }
  n105 = Add(
   A=Add(A=Power(Base=@n101.R, Exp=@n103.SourcePower), B=Multiply(A=@n102, B=@n103.FillScatter))
   B=Multiply(A=@n103.FillScatterIntensity, B=@n104.R)
  )
  n106 = Add(A=Add(A=@n103.FillScatter, B=@n103.FillScatter))
  n107 = VectorParameter() {
   desc="This set of parameters is used to control how the lighting texture is masked in the material.  The lightning mask bias determines at what positions in the CloudLayout texture alpha channel lightning can appear.  The lightning mask strength is multiplied against the biased resut to compensate for lightning masks which are either too opaque or transparent.  The low altitude texture weight determines the influence of the noise texture result on the lighting mask, where higher values increase influence of the texture."
   group=02 - Storm
   parameterName=Storm_LightningMasks
   sortPriority=50
   channelNames {a=N / A, b=CloudTextureWeight, g=LightningMaskStrength, r=LightningMaskBias}
   defaultValue=<rgba 10 5 0.5 1>
  }
  n108 = Saturate(None=Multiply(A=Subtract(A=@n7.A, B=@n107.LightningMaskBias), B=@n107.LightningMaskStrength))
  n109 = LinearInterpolate(A=@n83, B=ComponentMask(None=VolumetricAdvancedMaterialInput().ConservativeDensity as Float4) {r=true}, Alpha=@n107.CloudTextureWeight)
  n110 = Multiply(
   A=Multiply(A=Divide(A=@n105, B=@n106), B=Multiply(A=@n108, B=@n109) {constA=1})
   B=LinearInterpolate(A=Constant3Vector(), B=Multiply(A=@n91, B=@n92), Alpha=@n93)
  )
  ShadingPathSwitch(Default=@n110, Deferred=@n110, Mobile=Constant())
  n111 = LinearInterpolate(A=@n67.AmbientOcclusion, B=Multiply(A=@n68.StormOcclusionFactor, B=@n67.AmbientOcclusion) {constB=0.333}, Alpha=@n88)
  ShadingPathSwitch(Default=@n111, Deferred=@n111, Mobile=@n67.AmbientOcclusion)
  ShadingPathSwitch(Default=@n86, Deferred=@n86, Mobile=Constant() {r=1})
  n112 = VectorParameter() {
   desc="The cloud scattering uses a combination of two phase functions to determine how light bounces in the material.  Phase A and B determine the directionality of the light scattering, where negative values scatter backward toward the source and positive values scatter forward, away from the source. These values are defined in the range [-1, 1].  The phase blend determines how strong the contribution of each phase is, where 0 is entirely Phase A, and 1 is entirely Phase B. "
   group=03 - Multiscatter
   parameterName=Phase_Controls
   sortPriority=0
   channelNames {a=N / A, b=Phase Blend, g=Phase B, r=Phase A}
   defaultValue=<rgba 0.8 0.1667 0.575 1>
  }
  n113 = VectorParameter() {
   desc="This set of parameters controls the multiscattering approximations in the cloud material.  Intensity controls the overall stregth of the effect, occlusion determines how strongly light is attenuated, and eccentricity determines how uniform the scttering is.  Each of these parameters are valid in the [0, 1] range. "
   group=03 - Multiscatter
   parameterName=Multiscatter_Controls
   sortPriority=5
   channelNames {a=N / A, b=Eccentricity, g=Occlusion, r=Intensity}
   defaultValue=<rgba 0.6667 0.25 0.18 1>
  }
  VolumetricAdvancedMaterialOutput!(
   PhaseG=@n112.Phase A
   PhaseG2=@n112.Phase A
   PhaseBlend=@n112.Phase A
   MultiScatteringContribution=@n113.Intensity
   MultiScatteringOcclusion=@n113.Intensity
   MultiScatteringEccentricity=@n113.Intensity
   ConservativeDensity=MaterialFunctionCall(X=NamedRerouteUsage(), Y=NamedRerouteUsage(), Z=NamedRerouteUsage(), A=NamedRerouteUsage()) {materialFunction=&MakeFloat4}.Result
  ) {multiScatteringApproximationOctaveCount=2}
 ]
}
